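# Physics asset for the Doggus Sword melee weapon's skeletal mesh, giving it box-only collision bodies and joints.
# A central Anim body anchors three others: a blade built from one long box plus two angled boxes near its base, and left and right guard bodies.
# Each connects directly to Anim by a constraint, with both guard joints explicitly free in swing and twist, unlike the blade joint.
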PhysicsAsset /Figment_S09/Athena/Prototype/Blueprints/Doggus_Sword/Meshes/Doggus_Sword_Melee_Physics
bodies {
 Anim [{shapeType=Box, extentX=10.5651, extentY=5.6156, extentZ=31.9581, center=<0.3445 -0.0617 6.561>, rotation=<rot -0.4722 -90 -1.0151>}]
 blade [
  {shapeType=Box, extentX=4.7623, extentY=11.4877, extentZ=88.4563, center=<0.0249 0 42.7703>, rotation=<rot 0 -90 0>}
  {shapeType=Box, extentX=11.5, extentY=6.5, extentZ=19, center=<4.9545 0.0001 7.1425>, rotation=<rot 20.4758 0 0>}
  {shapeType=Box, extentX=11.5, extentY=6.5, extentZ=19, center=<-4.5324 0.0001 6.4708>, rotation=<rot 20.4757 179.3965 0>}
 ]
 guard_l [{shapeType=Box, extentX=6.9135, extentY=7.6622, extentZ=16.7068, center=<-0.397 5.5878 0.0034>, rotation=<rot 0 161.9444 -89.6404>}]
 guard_r [{shapeType=Box, extentX=6.4135, extentY=6.6622, extentZ=16.7068, center=<0.2791 -5.5493 0.0034>, rotation=<rot 0.0157 -20.5594 -89.6406>}]
}
constraints [
 {bone1Name=blade, bone2Name=Anim, swing1LimitDegrees=45, swing2LimitDegrees=45, twistLimitDegrees=45}
 {
  bone1Name=guard_r
  bone2Name=Anim
  swing1LimitDegrees=45
  swing1Motion=Free
  swing2LimitDegrees=45
  swing2Motion=Free
  twistLimitDegrees=45
  twistMotion=Free
 }
 {
  bone1Name=guard_l
  bone2Name=Anim
  swing1LimitDegrees=45
  swing1Motion=Free
  swing2LimitDegrees=45
  swing2Motion=Free
  twistLimitDegrees=45
  twistMotion=Free
 }
]
